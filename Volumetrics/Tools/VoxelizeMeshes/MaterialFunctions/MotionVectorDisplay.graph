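# Material function for the mesh voxelization tool that overrides incoming material attributes for motion vector display and UV-unwrap capture passes.
# MotionVectors mutes shading and outputs dilated, unpacked motion vectors as emissive.
# Unwrap uses WPO to flatten meshes by UV1 into an indexed grid cell around CapturePos.
# It writes WPO-displaced, volume-normalized world position to emissive, quantized to integers and packed via IntToHalfFloat.
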MaterialFunction /Volumetrics/Tools/VoxelizeMeshes/MaterialFunctions/MotionVectorDisplay
graph {nodes [
  n1 = FunctionInput() {inputType=FunctionInput_MaterialAttributes}
  BlendMaterialAttributes(A=@n1)
  n2 = BreakMaterialAttributes!(Attr=@n1)
  n3 = Add(A=@n2.WorldPositionOffset, B=WorldPosition() {worldPositionShaderOffset=WPT_ExcludeAllShaderOffsets}.XYZ)
  n4 = VectorParameter() {parameterName=CapturePos}
  n5 = Add(A=Divide(A=Subtract(A=@n3, B=@n4.RGB), B=ScalarParameter() {defaultValue=256, parameterName=VolumeWorldSize}) {desc=-0.5 to 0.5}) {constB=0.5, desc=0 to 1}
  n6 = Floor(None=Multiply(A=Saturate(None=@n5)) {constB=30719})
  Divide(A=@n6) {constB=30719}
  n7 = ScalarParameter() {parameterName=MotionVectors}
  n8 = ScalarParameter() {parameterName=Unwrap, sliderMax=1}
  n9 = Saturate(None=Multiply(A=OneMinus(None=@n7), B=OneMinus(None=@n8)) {desc=Or})
  n10 = MaterialFunctionCall() {materialFunction=&/UnPackMotionVectorsWithDilate}
  n11 = Add(
   A=Multiply(A=@n2.EmissiveColor, B=@n9)
   B=Multiply(A=Multiply(A=@n7, B=@n10.Result), B=ScalarParameter() {defaultValue=1, parameterName=MVMultiply}) {constA=1}
  )
  n12 = MaterialFunctionCall(Float3=@n6) {materialFunction=&BreakOutFloat3Components}
  n13 = Multiply(
   A=ScalarParameter() {parameterName=Unwrap, sliderMax=1}
   B=MaterialFunctionCall(
    X=MaterialFunctionCall!(Int=@n12.R) {materialFunction=&/IntToHalfFloat}.Result
    Y=MaterialFunctionCall(Int=@n12.G) {materialFunction=&IntToHalfFloat}.Result
    Z=MaterialFunctionCall(Int=@n12.B) {materialFunction=&IntToHalfFloat}.Result
   ) {materialFunction=&MakeFloat3}.Result
  ) {desc=Emissive}
  n14 = Multiply(A=@n2.OpacityMask)
  n15 = TextureCoordinate() {coordinateIndex=1}
  n16 = ScalarParameter() {defaultValue=500, parameterName=Size}
  n17 = ScalarParameter() {defaultValue=1, parameterName=XYFrames}
  n18 = MaterialFunctionCall!(1D Index=ScalarParameter() {parameterName=idx}, Cells XY=ScalarParameter() {defaultValue=1, parameterName=XYFrames}) {materialFunction=/Engine/Functions/Engine_MaterialFunctions01/Coordinates/1Dto2DIndex}
  n19 = Add(
   A=Multiply(A=ConstantBiasScale(None=@n15) {bias=-0.5, scale=1}, B=Divide(A=@n16, B=@n17))
   B=Multiply(A=ScalarParameter() {defaultValue=500, parameterName=Size}, B=Subtract(A=@n18.Centered UV Position) {constA=0.5, constB=0.5})
  )
  n20 = Constant()
  n21 = Add(A=Multiply(A=WorldPosition().XYZ) {constB=-1}, B=Add(A=AppendVector(A=@n19, B=@n20), B=VectorParameter() {parameterName=CapturePos}.RGB))
  n22 = ScalarParameter() {parameterName=Unwrap, sliderMax=1}
  FunctionOutput(None=MakeMaterialAttributes!(
    BaseColor=Multiply(A=@n2.BaseColor, B=@n9)
    Metallic=Multiply(A=@n2.Metallic, B=@n9)
    Specular=Multiply(A=@n2.Specular, B=@n9)
    Roughness=Multiply(A=@n2.Roughness, B=@n9)
    Anisotropy
    EmissiveColor=Add(A=@n11, B=@n13) {desc=WorldPos}
    Opacity=Multiply(A=@n2.Opacity, B=@n9)
    OpacityMask=Saturate(None=@n14)
    Normal=Multiply(A=@n2.Normal, B=@n9)
    Tangent
    WorldPositionOffset=LinearInterpolate(A=@n2.WorldPositionOffset, B=@n21, Alpha=@n22)
    SubsurfaceColor=Multiply(A=@n2.SubsurfaceColor, B=@n9)
    ClearCoat
    ClearCoatRoughness
    AmbientOcclusion=Multiply(A=@n2.AmbientOcclusion, B=@n9)
    Refraction
    CustomizedUVs_0
    CustomizedUVs_1
    CustomizedUVs_2
    CustomizedUVs_3
    CustomizedUVs_4
    CustomizedUVs_5
    CustomizedUVs_6
    CustomizedUVs_7
    PixelDepthOffset
    ShadingModel
    Displacement
   ))
  Multiply(A=@n22, B=@n21) {desc=Unwrap WPO}
  Saturate(None=Add(A=@n7, B=@n8)) {desc="Either pass == 1"}
 ]}
