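# Example unlit material for the Volumetrics mesh voxelization tools, with animated world position offset and motion vector display.
# A DebugTimeSine wave, driven by a speed parameter, scales a vector parameter (Z axis by default) to form the WPO, while a color parameter feeds emissive.
# These material attributes pass through the MotionVectorDisplay function, and negative emissive values are allowed.
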Material /Volumetrics/Tools/VoxelizeMeshes/Materials/StaticMeshMats/M_Example_WPO_Force_Mat
graph {
 properties {
  bAllowNegativeEmissiveColor=true
  bUseMaterialAttributes=true
  pixelDepthOffsetMode=PDOM_Legacy
  shadingModel=MSM_Unlit
  usedShadingModels=Unlit
 }
 nodes [
  n1 = MaterialFunctionCall(Speed=ScalarParameter() {defaultValue=1, parameterName=speed}) {materialFunction=&DebugTimeSine}
  n2 = VectorParameter() {parameterName=WPO Amount, defaultValue=<rgba 0 0 100 0>}
  MaterialFunctionCall!(In=MakeMaterialAttributes!(
    BaseColor
    Metallic
    Specular
    Roughness
    Anisotropy
    EmissiveColor=VectorParameter() {parameterName=Color, defaultValue=<rgba 0 1 0 0>}.RGB
    Opacity
    OpacityMask
    Normal
    Tangent
    WorldPositionOffset=Multiply(A=@n1.Result, B=@n2.RGB)
    SubsurfaceColor
    ClearCoat
    ClearCoatRoughness
    AmbientOcclusion
    Refraction
    CustomizedUVs_0
    CustomizedUVs_1
    CustomizedUVs_2
    CustomizedUVs_3
    CustomizedUVs_4
    CustomizedUVs_5
    CustomizedUVs_6
    CustomizedUVs_7
    PixelDepthOffset
    ShadingModel
    Displacement
   )) {materialFunction=/Volumetrics/Tools/VoxelizeMeshes/MaterialFunctions/MotionVectorDisplay}
 ]
}
